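# Physics asset for the Firework backpack cosmetic mesh, defining capsule bodies and joints for the pack and its dangling strap.
# A large capsule covers the pack, with small capsules on the root, strap base and five dynamic strap bones, the last strap capsule being larger.
# Constraints chain the strap with locked twist and varying swing limits, while base-to-pack swings freely.
# Pack and root carry two constraints, one free and one tightly swing-limited.
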PhysicsAsset /BRCosmetics/Accessories/FORT_Backpacks/Backpack_F_MED_Firework/Meshes/F_MED_Firework_Pack_Physics
bodies {
 base_strap [{shapeType=Capsule, length=0.505, radius=0.505, shapeName=base_strap_capsule, rotation=<rot -90 0 0>}]
 dyn_strap_1 [{shapeType=Capsule, length=0.505, radius=0.505, shapeName=dyn_strap_1_capsule, center=<1.0886 0 0>, rotation=<rot -90 0 0>}]
 dyn_strap_2 [{shapeType=Capsule, length=0.505, radius=0.505, shapeName=dyn_strap_2_capsule, center=<0.4561 0 0>, rotation=<rot -90 0 0>}]
 dyn_strap_3 [{shapeType=Capsule, length=0.505, radius=0.505, shapeName=dyn_strap_3_capsule, center=<0.5628 0 0>, rotation=<rot -90 0 0>}]
 dyn_strap_4 [{shapeType=Capsule, length=0.505, radius=0.505, shapeName=dyn_strap_4_capsule, center=<1.0848 0 0>, rotation=<rot -90 0 0>}]
 dyn_strap_5 [{shapeType=Capsule, length=3.7238, radius=1.3175, shapeName=dyn_strap_5_capsule, center=<2.0888 0 0>, rotation=<rot -70 0.0003 0>}]
 pack [{shapeType=Capsule, length=18.5312, radius=4.5938, shapeName=pack_capsule, center=<-0.0001 -5.7353 -12.1916>, rotation=<rot -90 0 0>}]
 root [{shapeType=Capsule, length=0.505, radius=0.505, shapeName=root_capsule, rotation=<rot 0 -90 0>}]
}
constraints [
 {
  bone1Name=base_strap
  bone2Name=pack
  swing1LimitDegrees=45
  swing1Motion=Free
  swing2LimitDegrees=45
  swing2Motion=Free
  twistLimitDegrees=15
  twistMotion=Free
 }
 {bone1Name=dyn_strap_1, bone2Name=base_strap, swing1LimitDegrees=15, swing2LimitDegrees=15, twistLimitDegrees=15, twistMotion=Locked}
 {bone1Name=dyn_strap_2, bone2Name=dyn_strap_1, swing1LimitDegrees=45, swing2LimitDegrees=45, twistLimitDegrees=15, twistMotion=Locked}
 {bone1Name=dyn_strap_3, bone2Name=dyn_strap_2, swing1LimitDegrees=80, swing2LimitDegrees=60, twistLimitDegrees=15, twistMotion=Locked}
 {bone1Name=dyn_strap_4, bone2Name=dyn_strap_3, swing1LimitDegrees=80, swing2LimitDegrees=70, twistLimitDegrees=15, twistMotion=Locked}
 {bone1Name=dyn_strap_5, bone2Name=dyn_strap_4, swing1LimitDegrees=10, swing2LimitDegrees=20, twistLimitDegrees=15, twistMotion=Locked}
 {
  bone1Name=pack
  bone2Name=root
  swing1LimitDegrees=45
  swing1Motion=Free
  swing2LimitDegrees=45
  swing2Motion=Free
  twistLimitDegrees=45
  twistMotion=Free
 }
 {bone1Name=pack, bone2Name=root, swing1LimitDegrees=1, swing2LimitDegrees=1, twistLimitDegrees=15, twistMotion=Locked}
]
